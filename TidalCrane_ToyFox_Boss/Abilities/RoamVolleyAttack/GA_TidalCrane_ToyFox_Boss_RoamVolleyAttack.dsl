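; Toy Fox boss roam volley attack ability that overrides targeting, volley type and spawn point of an inherited projectile volley.
; It asks the boss's current roam target actor for one set of locations sized by the projectile count, stores them, then calls SpawnProjectiles.
; The volley type is fixed to a single enum value.
; A hotfix scalable float switches spawning to the mesh's ProjectileVolleySocket, otherwise the parent's location applies.
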(event EventFindRandomProjectileTargetLocations
  ;[0 0.15 0.03864 0.5] Find projectile target locations by asking the roam target.
  (Utilities|Array|Clear (Variables|InternalData|GetVolleyTargetLocations))
  (bind _bossbga (Variables|InternalData|GetBossBGA)
    (:then
      (bind _currenttargetactor (Default|GetCurrentTargetActor(Message) _bossbga))
      (bind _asb_tidal_crane_toy_fox_roam_target (Utilities|Casting|CastToB_TidalCraneToyFoxRoamTarget _currenttargetactor))
      (bind _locations (Class|BTidalCraneToyFoxRoamTarget|GetTargetLocations _asb_tidal_crane_toy_fox_roam_target (ScalableFloat|GetValueasInteger (Variables|Default|GetProjectileCount))))
      (Variables|InternalData|SetVolleyTargetLocations _locations)
      (CallFunction|SpawnProjectiles))
    (:else)))

(fn ChooseProjectileVolleyType
  (return "NewEnumerator0"))

(fn GetProjectileSpawnLocation
  ;[0 0.1229 0.15 0.5] GetProjectileSpawnLocation: Spawn the projectile from the socket location or from an offset location based on hotfix.
  (Utilities|FlowControl|Sequence
    (:then_0
      (if (> (ScalableFloat|GetCachedScalableFloatValue (Variables|Hotfix|GetHotfix_SpawnProjectileFromSocket)) 0)
        (bind _asbga_tidal_crane_boss_toy_fox (Utilities|Casting|CastToBGA_TidalCrane_Boss_ToyFox (Ability|GetAvatarActorfromActorInfo)))
        (bind _skeletalmeshcomponentbudgeted (Variables|/TidalCrane_Swarm_Boss/Gameplay/BGA_TidalCrane_Boss.BGA_TidalCrane_Boss_C|GetSkeletalMeshComponentBudgeted _asbga_tidal_crane_boss_toy_fox)
          (:then
            (return (Transformation|GetSocketLocation _skeletalmeshcomponentbudgeted "ProjectileVolleySocket")))
          (:else))))
    (:then_1
      (return (|CallParentFunction)))))
